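# Sound cue for the ambient ring of a station payphone prop.
# A single wave player feeds the ring wave straight to the output, with no randomization or modulation.
# The cue uses payphone-specific attenuation, the loot ambient loop sound class and loot ambient concurrency.
# It plays at reduced volume with virtualization disabled.
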SoundCue /Game/Sounds/Fort_Prop_Sounds/Exterior/StationPayphone/Station_Payphone_Ambient_Ring_Cue
root=WavePlayer_1
nodes [{id=WavePlayer_1, properties {soundWaveAssetPtr=&/Station_Payphone_Ambient_Ring_01}}]
properties {
 attenuationSettings=/Game/Sounds/Attenuation/Fort_Payphone_Amb_Att
 maxDistance=1700
 soundClassObject=/Game/Sounds/FortSoundClassesAndModes/Fort_Loot_Ambient_Loop
 virtualizationMode=Disabled
 volumeMultiplier=0.4
 concurrencySet [/Game/Sounds/Concurrency_Rules/Loot_Ambient_Conc]
}
