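# Sound cue for interacting with the station payphone exterior prop.
# A single Wave Player plays the Station_Payphone_Interact wave directly, with no randomization or mixing nodes.
# It uses the Season 7 quest attenuation asset and the loot reward sound class, with virtualization disabled.
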SoundCue /Game/Sounds/Fort_Prop_Sounds/Exterior/StationPayphone/Station_Payphone_Interact_Cue
root=WavePlayer_0
nodes [{id=WavePlayer_0, properties {soundWaveAssetPtr=&/Station_Payphone_Interact}}]
properties {
 attenuationSettings=/Game/Sounds/Attenuation/Fort_Quest_Season7
 maxDistance=2400
 soundClassObject=&Fort_Loot_Reward_PSM
 virtualizationMode=Disabled
 volumeMultiplier=1
}
